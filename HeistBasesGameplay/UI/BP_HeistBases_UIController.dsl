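; Heist Bases UI controller that tracks base managers and resolves a POI tag to its base's state for UI use.
; On authority, once playlist data is ready, it caches registered managers by gameplay tag and binds for late registrations until EndPlay.
; GetCurrentStateFromPOI maps the POI to a base ID, finds the cached manager and returns it with a lockdown-based state, or a default state if none matches.
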(event EventBeginPlay
  (Networking|SwitchHasAuthority
    (:Authority
      (Class|FortAsyncActionCurrentPlaylistReady|PlaylistDataReadyAsync
        (:OnReady
          (CallFunction|CacheCurrentRegisteredManagers)
          ;[1 1 1] Register for late manager adds
          (bind _returnvalue (Class|FortGlobalActorRegistrySubsystem|GetActorRegistryDataforGameplayTag (WorldSubsystems|GetFortGlobalActorRegistrySubsystem) (Variables|Properties|GetBaseManagerRegistrationTag)))
          (Utilities|IsValid _returnvalue
            (:Is%20Valid
              (bind _output_get (Variables|Default|SetCachedDataRegistry _returnvalue))
              (EventDispatchers|BindEventtoOnActorRegistered _output_get (EventDispatchers|CreateEvent :!event CacheRegistedManager)))))))))

(event EventEndPlay
  (params (EndPlayReason :type EEndPlayReason%20Enum :default "Destroyed"))
  (bind _cacheddataregistry (Variables|Default|GetCachedDataRegistry)
    (:then
      (EventDispatchers|UnbindEventfromOnActorRegistered _cacheddataregistry (EventDispatchers|CreateEvent :!event CacheRegistedManager)))
    (:else)))

(fn CacheCurrentRegisteredManagers
  (bind _basemanagerregistrationtag (Variables|Properties|GetBaseManagerRegistrationTag))
  (bind _outactors (Class|FortGlobalActorRegistrySubsystem|GetRegisteredActors (WorldSubsystems|GetFortGlobalActorRegistrySubsystem) _basemanagerregistrationtag)) ; get all the current registered actors
  (for _array_element _outactors
    (CallFunction|CacheRegistedManager _basemanagerregistrationtag _array_element)))

(fn CacheRegistedManager
  (params (GameplayTag :type Gameplay%20Tag%20Structure%20%28by%20ref%29) (RegisteredActor :type Actor%20Object%20Reference))
  (bind _asbga_heist_base_manager (Utilities|Casting|CastToBGA_HeistBaseManager RegisteredActor))
  (Utilities|Array|AddUnique (Variables|Default|GetBaseManagers) _asbga_heist_base_manager))

(fn GetBaseManagerFromId
  (params (HeistBaseId :type Gameplay%20Tag%20Structure))
  (for _array_element (Variables|Default|GetBaseManagers)
    (if (== (Variables|/HeistBasesGameplay/Items/BGA_HeistBaseManager.BGA_HeistBaseManager_C|GetHeistBaseId%20Tag%20Const _array_element) (Variables|GetHeistBaseId))
      (return true _array_element)))
  (return false 0))

(fn GetCurrentStateFromPOI
  (params (POITag :type Gameplay%20Tag%20Structure))
  (bind _poitoheistbase (Variables|Properties|GetPOIToHeistBase))
  (bind (_value _returnvalue) (Utilities|Map|Find _poitoheistbase POITag))
  (if _returnvalue
    (bind (_isvalid _heistbasemanager) (Utility|GetBaseManagerfromId _value))
    (if _isvalid
      (return (select (or (Variables|/HeistBasesGameplay/Items/BGA_HeistBaseManager.BGA_HeistBaseManager_C|GetbLockdownExpired _heistbasemanager) (Variables|/HeistBasesGameplay/Items/BGA_HeistBaseManager.BGA_HeistBaseManager_C|GetbInLockdown _heistbasemanager)) "NewEnumerator1" "NewEnumerator3") _heistbasemanager)
      (else
        _))
    (else
      _))
  (return "NewEnumerator0" 0))
